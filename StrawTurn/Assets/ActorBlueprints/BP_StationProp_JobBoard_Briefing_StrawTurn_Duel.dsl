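; StrawTurn duel briefing job board: a spooky TV quest prop whose screen and spotlight react to player conversation.
; Dialog marker states switch between face and static screen materials with matching light colors, playing a voice line during conversation.
; On deactivation, the server sets bDeactivated and triggers cosmetic screen-off and light-off effects.
; PlayServiceUsedEffects restores those visuals after world-partition client recreation; the construction script overrides the parent's deactivation sound.
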(event Collision|EventActorBeginOverlap
  (params (OtherActor :type Actor%20Object%20Reference))
  (|CallParentFunction OtherActor) ; This node is disabled and will not be called.\nDrag off pins to build functionality.
)

(event EventDeactivateOnUse
  ;[0 0 0] Display screen states
  ;;[0.15 0.15 0.15] When machine is deactivated for everyone, HasJobs mesh will be hidden and NoJobs mesh will be visible. Material change to turn off the TV should be modified on the NoJobs static mesh component in this BP if possible. Any logic for this, if needed, can go here
  (|CallParentFunction)
  (Networking|SwitchHasAuthority
    (:Authority
      (Variables|Default|SetbDeactivated true)
      (CallFunction|PlayAdditionalDeactivationEffects) ; Deactivate on use is server only - need to call a replicated event here to do client only logic
    )))

(event EventTick
  (params (DeltaSeconds :type Float%20%28single-precision%29))
  (|CallParentFunction DeltaSeconds) ; This node is disabled and will not be called.\nDrag off pins to build functionality.
)

(event Custom|PlayAdditionalDeactivationEffects
  ;[0 0 0] Display screen states
  ;;[0.15 0.15 0.15] When machine is deactivated for everyone, HasJobs mesh will be hidden and NoJobs mesh will be visible. Material change to turn off the TV should be modified on the NoJobs static mesh component in this BP if possible. Any logic for this, if needed, can go here
  (Networking|CanExecuteCosmeticEvents
    (:True
      (Rendering|Material|SetMaterial (Variables|Default|GetMesh_HasJobs) 1 "/Game/Environments/Helios/Props/Quest/Quest_SpookyTV_A/Materials/M_Quest_SpookyTV_ScreenB_Off_Inst.M_Quest_SpookyTV_ScreenB_Off_Inst")
      (Rendering|SetVisibility (Variables|Default|GetSpotLight) false false))))

(event OnDialogMarkerInteractionStateChanged(NonPlayerConversationComponent)
  (params (CurrentInteractionState :type EDialogMarkerInteractionState%20Enum :default "Conversation") (InteractingController :type Fort%20Player%20Controller%20Object%20Reference))
  ;[0 0 0] Display screen states
  ;;[0.15 0.15 0.15] Change the screen state when the player is conversing with the machine
  (if (not (Variables|Default|GetbDeactivated))
    (Networking|CanExecuteCosmeticEvents
      (:True
        (switch Utilities|FlowControl|Switch|SwitchonEDialogMarkerInteractionState CurrentInteractionState
          (:Conversation
            ;[0.15 0.15 0.15] Player actively conversing
            (Rendering|Material|SetMaterial (Variables|Default|GetMesh_HasJobs) 1 "/Game/Environments/Helios/Props/Quest/Quest_SpookyTV_A/Materials/M_Quest_SpookyTV_ScreenB_Face_Inst.M_Quest_SpookyTV_ScreenB_Face_Inst")
            (Audio|SpawnSoundAttached (Variables|Audio|GetStrawTurnDeactivateVO) (Variables|Default|GetStrawTurn_JobBoard_Accept_VO_Cue))
            (Rendering|SetVisibility (Variables|Default|GetSpotLight) true false)
            (Rendering|Components|Light|SetLightColor (Variables|Default|GetSpotLight) (Variables|Default|GetLight%20Color%20%28Face%29) true))
          (:InteractionRange
            _)
          (:IndicatorRange
            _))
        ;[0.15 0.15 0.15] Player is within conversation range, but not conversing
        (Rendering|Material|SetMaterial (Variables|Default|GetMesh_HasJobs) 1 "/Game/Environments/Helios/Props/Quest/Quest_SpookyTV_A/Materials/M_Quest_SpookyTV_ScreenB.M_Quest_SpookyTV_ScreenB")
        (Rendering|SetVisibility (Variables|Default|GetSpotLight) true false)
        (Rendering|Components|Light|SetLightColor (Variables|Default|GetSpotLight) (Variables|Default|GetLight%20Color%20%28Static%29) true)))))

(fn PlayServiceUsedEffects
  (params (PlayerController :type Fort%20Player%20Controller%20Object%20Reference))
  ;[0 0 0] FORT-795090 - world partitioning may tear down and recreate this actor client side. When beginning play, restore 'used' visuals if this machine is deactivated
  (|CallParentFunction PlayerController)
  (if (Variables|Default|GetbDeactivated)
    (Networking|CanExecuteCosmeticEvents
      (:True
        (Rendering|Material|SetMaterial (Variables|Default|GetMesh_HasJobs) 1 "/Game/Environments/Helios/Props/Quest/Quest_SpookyTV_A/Materials/M_Quest_SpookyTV_ScreenB_Off_Inst.M_Quest_SpookyTV_ScreenB_Off_Inst")
        (Rendering|SetVisibility (Variables|Default|GetSpotLight) false false)))))

(fn ConstructionScript
  ; Override the parent's deactivate sound
  (|CallParentFunction)
  (Variables|Effects|SetJobBoardDeactivatedSound (Variables|Audio|GetStrawTurnDeactivateSound)))
